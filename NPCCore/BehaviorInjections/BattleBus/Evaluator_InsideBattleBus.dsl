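; Behavior injection evaluator that reports whether an NPC is inside the Battle Bus.
; Both the starting and ongoing condition checks read a boolean blackboard value through the behavior tree component.
; The key comes from a GetIsInBusBlackboardKey variable, so both checks pass only while that flag is true.
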(fn EvaluateOngoingConditions
  (params (OwnerComp :type Behavior%20Tree%20Component%20Object%20Reference))
  (return (Blackboard|GetBool (Variables|Default|GetIsInBusBlackboardKey) OwnerComp)))

(fn EvaluateStartingConditions
  (params (OwnerComp :type Behavior%20Tree%20Component%20Object%20Reference))
  (return (Blackboard|GetBool (Variables|Default|GetIsInBusBlackboardKey) OwnerComp)))
